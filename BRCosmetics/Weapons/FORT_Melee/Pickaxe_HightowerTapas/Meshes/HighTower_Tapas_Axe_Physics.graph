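# Physics asset for the HighTower Tapas pickaxe axe mesh, defining physical bodies for its strap bones.
# Four capsule bodies form a chain from strap_01_MSTR through strap_01, strap_02 and strap_03.
# Each link is joined to the previous one by a constraint with moderate swing limits and a tight twist limit, so the strap can bend but barely twist.
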PhysicsAsset /BRCosmetics/Weapons/FORT_Melee/Pickaxe_HightowerTapas/Meshes/HighTower_Tapas_Axe_Physics
bodies {
 strap_01 [{shapeType=Capsule, length=1.3008, radius=1.2693, center=<1.0946 0.0291 -0.3418>, rotation=<rot 1.1721 91.1747 -90>}]
 strap_01_MSTR [{shapeType=Capsule, length=0.505, radius=0.505, rotation=<rot 0 -90 0>}]
 strap_02 [{shapeType=Capsule, length=2.0318, radius=1.4644, center=<0.543 0.0127 -0.2318>, rotation=<rot 0 -82.5336 -82.7356>}]
 strap_03 [{shapeType=Capsule, length=2.7357, radius=1.3297, center=<0.6369 -0.0128 -0.0699>, rotation=<rot 0 92.063 -89.7559>}]
}
constraints [
 {bone1Name=strap_01, bone2Name=strap_01_MSTR, swing1LimitDegrees=45, swing2LimitDegrees=45, twistLimitDegrees=5}
 {bone1Name=strap_02, bone2Name=strap_01, swing1LimitDegrees=45, swing2LimitDegrees=45, twistLimitDegrees=5}
 {bone1Name=strap_03, bone2Name=strap_02, swing1LimitDegrees=45, swing2LimitDegrees=45, twistLimitDegrees=5}
]
