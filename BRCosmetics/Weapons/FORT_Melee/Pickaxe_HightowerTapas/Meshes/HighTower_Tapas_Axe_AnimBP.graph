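# Animation Blueprint for the HighTower Tapas pickaxe that adds secondary motion to its strap bones.
# The pose copied from a mesh passes through DefaultSlot and feeds either a pre-rotated, limited Trail Controller (bRigidsEnabled true) or a RigidBody simulation (false).
# The DisableDynamics curve blends that result back to plain animation.
# An Anim Dynamics strap chain is unconnected, and the Anim bone's yaw flip has zero alpha.
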AnimBlueprint /BRCosmetics/Weapons/FORT_Melee/Pickaxe_HightowerTapas/Meshes/HighTower_Tapas_Axe_AnimBP
target_skeleton=&/HighTower_Tapas_Axe_Skeleton
graphs [
 AnimGraph {nodes [
   n1 = Anim Dynamics - Bone: strap_01(Alpha=1, ExternalForce="0,0,0") {
    angularDampingOverride=1
    angularSpringConstant=1500
    bAngularSpring=true
    bGravityOverrideInSimSpace=true
    bOverrideAngularDamping=true
    gravityScale=0.34
    simulationSpace=RootRelative
    alphaBoolBlend {blendOption=Linear}
    boundBone {boneName=strap_01}
    chainEnd {boneName=strap_03}
    componentAppliedLinearAccClamp=<5000 5000 5000>
    componentLinearAccScale=<0.4 0.4 0.7>
    componentLinearVelScale=<0.4 0.4 0.7>
    gravityOverride=<-980 -980 -980>
    physicsBodyDefinitions [{
      sphereCollisionRadius=0
      boundBone {boneName=strap_01}
      boxExtents=<10 5 5>
      constraintSetup {angularLimitsMax=<0 45 90>, angularLimitsMin=<0 -45 -90>, angularTarget=<0 0 0>}
      localJointOffset=<0.25 0 0>
     }]
   }
   n2 = Anim Dynamics - Bone: strap_02(ComponentPose=@n1, Alpha=1, ExternalForce="0,0,0") {
    angularDampingOverride=1
    angularSpringConstant=1500
    bAngularSpring=true
    bGravityOverrideInSimSpace=true
    bOverrideAngularDamping=true
    gravityScale=0.34
    simulationSpace=RootRelative
    alphaBoolBlend {blendOption=Linear}
    boundBone {boneName=strap_02}
    chainEnd {boneName=strap_03}
    componentAppliedLinearAccClamp=<5000 5000 5000>
    componentLinearAccScale=<0.4 0.4 0.7>
    componentLinearVelScale=<0.4 0.4 0.7>
    gravityOverride=<-980 -980 -980>
    physicsBodyDefinitions [{
      sphereCollisionRadius=0
      boundBone {boneName=strap_02}
      boxExtents=<10 5 5>
      constraintSetup {angularLimitsMax=<0 45 90>, angularLimitsMin=<0 -45 -90>, angularTarget=<0 0 0>}
      localJointOffset=<0.25 0 0>
     }]
   }
   Anim Dynamics - Bone: strap_01(ComponentPose=@n2, Alpha=1, ExternalForce="0,0,0") {
    angularDampingOverride=1
    angularSpringConstant=1500
    bAngularSpring=true
    bGravityOverrideInSimSpace=true
    bOverrideAngularDamping=true
    gravityScale=0.34
    simulationSpace=RootRelative
    alphaBoolBlend {blendOption=Linear}
    boundBone {boneName=strap_01}
    chainEnd {boneName=strap_03}
    componentAppliedLinearAccClamp=<5000 5000 5000>
    componentLinearAccScale=<0.4 0.4 0.7>
    componentLinearVelScale=<0.4 0.4 0.7>
    gravityOverride=<-980 -980 -980>
    physicsBodyDefinitions [{
      sphereCollisionRadius=0
      boundBone {boneName=strap_01}
      boxExtents=<8 5 5>
      constraintSetup {angularLimitsMax=<0 90 90>, angularLimitsMin=<0 -90 -90>, angularTarget=<0 0 0>}
      localJointOffset=<0.25 0 0>
     }]
   }
   n3 = "Transform (Modify) Bone - Bone: strap_01_MSTR"(ComponentPose=Local To Component(LocalPose=Use cached pose 'initCache'()), Alpha=1, Translation="0,0,0", Rotation="-60,0,0", Scale="1,1,1") {
    rotationMode=BMM_Additive
    rotationSpace=BCS_BoneSpace
    alphaBoolBlend {blendOption=Linear}
    boneToModify {boneName=strap_01_MSTR}
    rotation {pitch=-60}
   }
   n4 = Trail controller - Bone: strap_03!(ComponentPose=@n3, Alpha=1, RelaxationSpeedScale:float=1.5, FakeVelocity:Vector="0,0,0") -> (Pose) {
    bActorSpaceFakeVel=true
    bLimitRotation=true
    chainLength=3
    debugLifeTime=0.1
    maxDeltaTime=0.0333
    relaxationSpeedScale=1.2
    alphaScaleBiasClamp {bInterpResult=true, interpSpeedDecreasing=25, interpSpeedIncreasing=25, scale=1.5}
    baseJoint {boneName=Anim}
    planarLimits [
     {drivingBone {boneName=root}, planeTransform {location=<0 0 10>, rotation=<rot 0 0 0>, scale=<1 1 1>}}
     {drivingBone {boneName=Anim}, planeTransform {location=<0 40 0>, rotation=<rot 0 0 0>, scale=<1 1 1>}}
    ]
    rotationLimits [{limitMax=<0 0 30>, limitMin=<0 -30 -30>}, {limitMax=<0 0 30>, limitMin=<0 -30 -30>}, {limitMax=<180 180 180>}]
    rotationOffsets [<0 0 0>, <0 0 0>, <0 0 0>]
    trailBone {boneName=strap_03}
    trailRelaxationSpeed {editorCurveData {keys [<at 0 15 RCIM_Linear>,<at 0.5289 13.1838 RCIM_Linear>,<at 1 6 RCIM_Linear>]}}
   }
   n5 = Component To Local(ComponentPose=@n4)
   n6 = RigidBody!(ComponentPose=Local To Component(LocalPose=Use cached pose 'initCache'()), Alpha, OverrideWorldGravity:Vector="0,0,-980", ExternalForce) -> (Pose) {
    bClampLinearTranslationLimitToRefPose=true
    bOverrideWorldGravity=true
    overridePhysicsAsset=&/HighTower_Tapas_Axe_Physics
    simulationSpace=BaseBoneSpace
    alphaBoolBlend {blendOption=Linear}
    baseBoneRef {boneName=root}
    componentAppliedLinearAccClamp=<3000 3000 3000>
    componentLinearAccScale=<0.4 0.4 0.7>
    componentLinearVelScale=<0.5 0.5 0.7>
    overrideWorldGravity {z=-980}
   }
   n7 = Component To Local(ComponentPose=@n6)
   n8 = Get bRigidsEnabled() -> (bRigidsEnabled:bool)
   Output Pose(Result=Two Way Blend!(
     A=Blend Poses by bool(BlendPose_0=@n5, BlendPose_1=@n7, bActiveValue=@n8) {blendType=Linear}
     B=Use cached pose 'JustAnimation'()
     AlphaCurveName:Name=DisableDynamics
    ) -> (Pose) {blendNode {alphaInputType=Curve, alphaBoolBlend {blendOption=Linear}}}) {layerGroup=None}
   n9 = "Transform (Modify) Bone - Bone: Anim"(
    ComponentPose=Local To Component(LocalPose=Use cached pose 'JustAnimation'())
    Alpha=0
    Translation="0,0,0"
    Rotation="0,180,0"
    Scale="1,1,1"
   ) {
    rotationMode=BMM_Additive
    rotationSpace=BCS_BoneSpace
    alphaBoolBlend {blendOption=Linear}
    boneToModify {boneName=Anim}
    rotation {yaw=180}
   }
   Save cached pose 'initCache'(Pose=Component To Local(ComponentPose=@n9)) {cacheName=initCache}
   Save cached pose 'JustAnimation'(Pose=Slot 'DefaultSlot'(Source=Copy Pose From Mesh())) {cacheName=JustAnimation}
  ]}
 EventGraph {links [], nodes []}
]
variables []
